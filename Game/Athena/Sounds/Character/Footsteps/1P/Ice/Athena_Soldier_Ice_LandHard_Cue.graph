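# Sound cue for a first-person soldier's hard landing on ice in Athena, using snow landing samples.
# Each variant mixes a random soldier landing sound with a random snow step, under random pitch and volume modulation.
# A quality-level switch gives the first two levels six snow samples and the last two a two-sample set excluded from branch culling.
# Player-land attenuation, 1P foley sound class and footstep-landing concurrency govern playback.
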SoundCue /Game/Athena/Sounds/Character/Footsteps/1P/Ice/Athena_Soldier_Ice_LandHard_Cue
root=Modulator_4
nodes [
 {id=Mixer_0, children [Random_1,Random_2], properties {inputVolume [0.6,0.9]}}
 {id=Mixer_2, children [Random_3,Random_0], properties {inputVolume [0.6,0.9]}}
 {id=Modulator_4, children [QualityLevel_0], properties {pitchMax=1.1, pitchMin=0.9, volumeMax=1.3, volumeMin=0.7}}
 {id=QualityLevel_0, children [Mixer_2,Mixer_2,Mixer_0,Mixer_0]}
 {id=Random_0, children [WavePlayer_0,WavePlayer_1,WavePlayer_6,WavePlayer_7,WavePlayer_8,WavePlayer_9], properties {weights [1,1,1,1,1,1]}}
 {id=Random_1, children [WavePlayer_2,WavePlayer_3], properties {bShouldExcludeFromBranchCulling=true, weights [1,1]}}
 {id=Random_2, children [WavePlayer_4,WavePlayer_5], properties {bShouldExcludeFromBranchCulling=true, weights [1,1]}}
 {id=Random_3, children [WavePlayer_19, WavePlayer_20], properties {weights [1,1]}}
 {id=WavePlayer_0, properties {soundWaveAssetPtr=/Game/Sounds/Fort_Step_Sounds/Soldier/Soldier_Step_Land_Snow_01}}
 {id=WavePlayer_1, properties {soundWaveAssetPtr=/Game/Sounds/Fort_Step_Sounds/Soldier/Soldier_Step_Land_Snow_02}}
 {id=WavePlayer_19, properties {soundWaveAssetPtr=/Game/Sounds/Fort_Step_Sounds/Human/soldier_land_01}}
 {id=WavePlayer_2, properties {soundWaveAssetPtr=&soldier_land_01}}
 {id=WavePlayer_20, properties {soundWaveAssetPtr=/Game/Sounds/Fort_Step_Sounds/Human/soldier_land_02}}
 {id=WavePlayer_3, properties {soundWaveAssetPtr=&soldier_land_02}}
 {id=WavePlayer_4, properties {soundWaveAssetPtr=&Soldier_Step_Land_Snow_01}}
 {id=WavePlayer_5, properties {soundWaveAssetPtr=&Soldier_Step_Land_Snow_02}}
 {id=WavePlayer_6, properties {soundWaveAssetPtr=/Game/Sounds/Fort_Step_Sounds/Soldier/Soldier_Step_Land_Snow_03}}
 {id=WavePlayer_7, properties {soundWaveAssetPtr=/Game/Sounds/Fort_Step_Sounds/Soldier/Soldier_Step_Land_Snow_04}}
 {id=WavePlayer_8, properties {soundWaveAssetPtr=/Game/Sounds/Fort_Step_Sounds/Soldier/Soldier_Step_Land_Snow_05}}
 {id=WavePlayer_9, properties {soundWaveAssetPtr=/Game/Sounds/Fort_Step_Sounds/Soldier/Soldier_Step_Land_Snow_06}}
]
properties {
 attenuationSettings=/Game/Sounds/Attenuation/Fort_Player_Land_Athena_Att
 maxDistance=2097152
 soundClassObject=&Fort_Player_Foley_1P_Athena
 volumeMultiplier=1.2
 concurrencyOverrides {maxCount=4, resolutionRule=PreventNew}
 concurrencySet [/Game/Sounds/Concurrency_Rules/Athena_Footstep_Land_Conc]
}
